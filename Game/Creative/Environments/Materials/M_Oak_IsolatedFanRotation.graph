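# Metal material for the Oak wind turbine fan in Creative, animating the blade rotation in the vertex shader.
# World position offset rotates vertices about the object pivot on local X, with a static switch choosing sine sway or constant spin.
# FixRotateAboutAxisNormals corrects the normal map for that rotation, while one mask channel drives metallic, specular and roughness.
# Attributes pass through Fortnite_BaseFunction as a gameplay asset, then MF_CreativeMoveTool.
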Material /Game/Creative/Environments/Materials/M_Oak_IsolatedFanRotation
graph {
 properties {
  bUseMaterialAttributes=true
  physMaterial=/Game/FortressPhysicalMaterials/Metal
  pixelDepthOffsetMode=PDOM_Legacy
  refractionMethod=RM_IndexOfRefraction
  usedShadingModels=Default Lit
 }
 nodes [
  n1 = TextureSample() {texture=/Game/Environments/Sets/Oak/Props/Textures/Original/T_Turbine_Wind_Fan_D}
  n2 = TextureSample() {samplerType=SAMPLERTYPE_Masks, texture=/Game/Environments/Sets/Oak/Props/Textures/Original/T_Turbine_Wind_Fan_S}
  n3 = Normalize(VectorInput=Transform(None=Constant3Vector() {constant=<rgba 1 0 0 0>}) {transformSourceType=TRANSFORMSOURCE_Local})
  n4 = Time()
  n5 = Multiply(A=@n4, B=ScalarParameter() {defaultValue=0.1, parameterName=SwayDistance})
  n6 = StaticSwitchParameter(
   True=Multiply(A=Sine(None=@n5), B=ScalarParameter() {defaultValue=0.5, parameterName=SwaySpeed})
   False=Multiply(A=@n4, B=ScalarParameter() {defaultValue=-1, parameterName=Speed})
  ) {parameterName=Sway or Constant Spin?}
  n7 = TextureSampleParameter2D() {parameterName=Normals, samplerType=SAMPLERTYPE_Normal, texture=/Game/Environments/Sets/Oak/Props/Textures/T_OAK_Turbine_Wind_Fan_N}
  n8 = MaterialFunctionCall(Rotation Axis=@n3, Rotation Angle=@n6, World Space Vertex Normals=Transform(None=@n7.RGB)) {materialFunction=&FixRotateAboutAxisNormals}
  n9 = RotateAboutAxis(
   NormalizedRotationAxis=@n3
   RotationAngle=@n6
   PivotPoint=MaterialFunctionCall() {materialFunction=&ObjectPivotPoint}.Object Pivot Location
   Position=WorldPosition().XYZ
  )
  MaterialFunctionCall(MA In=MaterialFunctionCall!(
    Material attributes=MakeMaterialAttributes!(
     BaseColor=@n1.RGB
     Metallic=@n2.G
     Specular=@n2.G
     Roughness=@n2.G
     Anisotropy
     EmissiveColor
     Opacity
     OpacityMask
     Normal=@n8.New Tangent Space Vertex Normal
     Tangent
     WorldPositionOffset=@n9
     SubsurfaceColor
     ClearCoat
     ClearCoatRoughness
     AmbientOcclusion
     Refraction
     CustomizedUVs_0
     CustomizedUVs_1
     CustomizedUVs_2
     CustomizedUVs_3
     CustomizedUVs_4
     CustomizedUVs_5
     CustomizedUVs_6
     CustomizedUVs_7
     PixelDepthOffset
     ShadingModel
     Displacement
    )
    IsGameplayAsset=StaticBool() {value=true}
   ) {materialFunction=/Game/Packages/Fortress_SharedMaterials/Base_Material_Functions/Fortnite_BaseFunction}.Material Attributes) {materialFunction=&MF_CreativeMoveTool}
 ]
}
